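# Niagara effect for the Sunshine backpack cosmetic, adding a glow, light and sparks at sampled skeletal mesh locations.
# Light holds a looping particle whose sprite and light flicker through shared curve shapes, alongside a mesh renderer.
# FN_RadialSprite repeatedly spawns orange sprites and meshes that grow and fade, using dynamic radial material parameters.
# Two color-shifting spark emitters push outward; Sparks_World adds gravity, inherited movement and secondary color/size attributes.
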NiagaraSystem /BRCosmetics/Accessories/FORT_Backpacks/F_MED_Sunshine_Pack/FX/NS_Backpack_Sunshine
dependencies {
 usedDataInterfaces [&NiagaraDataInterfaceColorCurve, &NiagaraDataInterfaceCurve, &NiagaraDataInterfaceVector2DCurve]
 usedDynamicInputs []
 usedModules [&SystemState]
}
emitters {
 FN_RadialSprite {
  emitterUpdateScript [
   &EmitterState()
   &SpawnRate(SpawnRate=3, Use Spawn Probability=false)
  ]
  particleSpawnScript [
   &InitializeParticle(
    Color=<rgba 5 5 5 1>
    Lifetime=2
    Mesh Scale Mode=::NewEnumerator3
    Sprite Rotation Mode=::NewEnumerator2
    Sprite Size Mode=::NewEnumerator3
    Uniform Sprite Size=30
    UsePositionOffset=false
    Write Material Random=true
   )
   /SetVariables(Particles.SpriteFacing=<0 -1 0>) {isSetParameters=true}
  ]
  particleUpdateScript [
   &SkeletalMeshLocation(MaskSpawnGroup=false, Mesh Sampling Type=::NewEnumerator5, Skeletal Mesh=$User.New Skeletal Mesh, Socket Sampling Mode=::NewEnumerator5)
   &ParticleState()
   &FN_Radial_Material(Base Sprite Tile U=0.75, Base Sprite Tile V=5, Depth Fade=5, Radial Mask Power=3, Sprite Extrusion Speed=0.05)
   /SetVariables(Particles.DynMatParms2:Vector4f=&MakeVector4(X=0.5, Y=0.5, Z=0, W=0)) {isSetParameters=true}
   /SetVariables(Particles.DynMatParms3:Vector4f=&MakeVector4(X=0.6, Y=4, Z=1, W=1)) {isSetParameters=true}
   &ScaleSpriteSize(Uniform Curve Sprite Scale=DataInterface() {propertyValues {Curve {keys [
        {arriveTangent=1.8607, interpMode=RCIM_Cubic, leaveTangent=1.8607, tangentMode=RCTM_User, time=0, value=0}
        {interpMode=RCIM_Cubic, tangentMode=RCTM_User, time=1, value=1}
       ]}}})
   &ScaleColor(
    Scale Alpha=&FloatFromCurve(FloatCurve=DataInterface() {propertyValues {Curve {keys [{interpMode=RCIM_Cubic, tangentMode=RCTM_User, time=0, value=1}, {interpMode=RCIM_Cubic, tangentMode=RCTM_User, time=1, value=0}]}}})
    Scale RGB=&MakeVectorFromLinearColorRGB(Color=&ColorFromCurve(ColorCurve=DataInterface() {propertyValues {
        AlphaCurve {keys [<at 0 1 RCIM_Cubic>]}
        BlueCurve {keys [{arriveTangent=-0.0275, interpMode=RCIM_Cubic, leaveTangent=-0.0275, time=0, value=0.1629}]}
        GreenCurve {keys [{arriveTangent=1.0015, interpMode=RCIM_Cubic, leaveTangent=1.0015, tangentMode=RCTM_User, time=0, value=0.5454}]}
        RedCurve {keys [{arriveTangent=-0.9724, interpMode=RCIM_Cubic, leaveTangent=-0.9724, tangentMode=RCTM_User, time=0, value=1}]}
       }}))
    ScaleA=true
    ScaleRGB=true
   )
   &FN_Mesh_Attributes(Create Position Attribute=true, Position:Vector3f=<0 -10 0>, Write to Module Namespace=true) {enabled=false}
   &ScaleMeshSize(Scale Factor=&VectorFromFloat(Value=&FloatFromCurve(
      FloatCurve=DataInterface() {propertyValues {Curve {keys [
          {arriveTangent=0.9899, interpMode=RCIM_Cubic, leaveTangent=0.9899, tangentMode=RCTM_User, time=0, value=0}
          {interpMode=RCIM_Cubic, tangentMode=RCTM_User, time=1, value=1}
         ]}}}
      Scale Curve=1.6
     )))
  ]
  rendererClasses [&NiagaraSpriteRendererProperties, &NiagaraMeshRendererProperties]
  renderers [
   {rendererClass=&NiagaraSpriteRendererProperties}
   {rendererClass=&NiagaraMeshRendererProperties, rendererIndex=1}
   {rendererClass=&NiagaraMeshRendererProperties, rendererIndex=2}
  ]
 }
 Light {
  emitterUpdateScript [
   &EmitterState(Life Cycle Mode=::NewEnumerator1, Loop Behavior=::NewEnumerator1, Loop Duration Mode=::NewEnumerator1, UseLoopDelay=false)
   &SpawnBurst_Instantaneous(Use Spawn Probability=false, UseLoopCountLimit=false)
  ]
  particleSpawnScript [
   &InitializeParticle(
    Color=<rgba 1 0.6736 0.2032 0.35>
    Lifetime=2
    Mesh Scale Mode=::NewEnumerator3
    Sprite Size Mode=::NewEnumerator3
    Uniform Sprite Size=1
    UsePositionOffset=false
    Write Material Random=true
   )
   &InitialMeshOrientation(Mesh Orientation Mode=::NewEnumerator1, Rotation=<-0.05 -0.25 0>, Use Rotation Vector=true) {enabled=false}
   &SkeletalMeshLocation(Bone Sampling Mode=::NewEnumerator4, MaskSpawnGroup=false, Skeletal Mesh=$User.New Skeletal Mesh)
  ]
  particleUpdateScript [
   &ParticleState(
    Kill Particles When Lifetime Has Elapsed=false
    Let Infinitely Lived Particles Die When Emitter Deactivates=true
    Loop Particles Lifetime=true
   )
   &CameraOffset(Camera Offset Amount=1)
   &ScaleColor(
    Scale RGB=&MakeVectorFromLinearColorRGB(Color=&ColorFromCurve(ColorCurve=DataInterface() {propertyValues {
        AlphaCurve {keys [<at 0 1 RCIM_Cubic>]}
        BlueCurve {keys [
          {arriveTangent=-0.5189, interpMode=RCIM_Cubic, leaveTangent=-0.5189, time=0, value=0.2756}
          {arriveTangent=0.0146, interpMode=RCIM_Cubic, leaveTangent=0.0146, time=0.1395, value=0.2032}
          <at 0.3366 0.2805>
          <at 0.6045 0.2032>
          <at 0.8318 0.2283>
          {arriveTangent=0.1178, interpMode=RCIM_Cubic, leaveTangent=0.1178, time=1, value=0.2481}
         ]}
        GreenCurve {keys [
          {arriveTangent=1.0015, interpMode=RCIM_Cubic, leaveTangent=1.0015, tangentMode=RCTM_User, time=0, value=0.8072}
          {arriveTangent=0.0733, interpMode=RCIM_Cubic, leaveTangent=0.0733, time=0.1395, value=0.6736}
          <at 0.3366 0.8319>
          <at 0.6045 0.6736>
          <at 0.8318 0.8964>
          {arriveTangent=-0.7397, interpMode=RCIM_Cubic, leaveTangent=-0.7397, tangentMode=RCTM_User, time=1, value=0.8402}
         ]}
        RedCurve {keys [
          {arriveTangent=-0.9724, interpMode=RCIM_Cubic, leaveTangent=-0.9724, tangentMode=RCTM_User, time=0, value=1}
          <at 0.1395 1 RCIM_Cubic>
          <at 0.3366 1>
          <at 0.6045 1>
          <at 0.8318 1>
          {arriveTangent=0.7691, interpMode=RCIM_Cubic, leaveTangent=0.7691, tangentMode=RCTM_User, time=1, value=1}
         ]}
       }}))
    ScaleA=true
    ScaleRGB=true
   ) {enabled=false}
   &FN_Sprite_Attributes(
    Color=&ColorFromCurve(ColorCurve=DataInterface() {propertyValues {
       AlphaCurve {keys [<at 0 1 RCIM_Cubic>]}
       BlueCurve {keys [
         {arriveTangent=-0.5189, interpMode=RCIM_Cubic, leaveTangent=-0.5189, time=0, value=0.2756}
         {arriveTangent=0.0146, interpMode=RCIM_Cubic, leaveTangent=0.0146, time=0.1395, value=0.2032}
         <at 0.3366 0.2805>
         <at 0.6045 0.2032>
         <at 0.8318 0.2283>
         {arriveTangent=0.1178, interpMode=RCIM_Cubic, leaveTangent=0.1178, time=1, value=0.2481}
        ]}
       GreenCurve {keys [
         {arriveTangent=1.0015, interpMode=RCIM_Cubic, leaveTangent=1.0015, tangentMode=RCTM_User, time=0, value=0.8072}
         {arriveTangent=0.0733, interpMode=RCIM_Cubic, leaveTangent=0.0733, time=0.1395, value=0.6736}
         <at 0.3366 0.8319>
         <at 0.6045 0.6736>
         <at 0.8318 0.8964>
         {arriveTangent=-0.7397, interpMode=RCIM_Cubic, leaveTangent=-0.7397, tangentMode=RCTM_User, time=1, value=0.8402}
        ]}
       RedCurve {keys [
         {arriveTangent=-0.9724, interpMode=RCIM_Cubic, leaveTangent=-0.9724, tangentMode=RCTM_User, time=0, value=1}
         <at 0.1395 1 RCIM_Cubic>
         <at 0.3366 1>
         <at 0.6045 1>
         <at 0.8318 1>
         {arriveTangent=0.7691, interpMode=RCIM_Cubic, leaveTangent=0.7691, tangentMode=RCTM_User, time=1, value=1}
        ]}
      }})
    Create Color Attribute=true
    Create SpriteSize Attribute=true
    SpriteSize=&Vector2DFromFloat(Value=&FloatFromCurve(
      FloatCurve=DataInterface() {propertyValues {Curve {keys [
          {arriveTangent=-0.5, interpMode=RCIM_Cubic, leaveTangent=-0.5, time=0, value=1}
          {arriveTangent=-0.25, interpMode=RCIM_Cubic, leaveTangent=-0.25, time=0.1, value=0.95}
          {arriveTangent=0.5, interpMode=RCIM_Cubic, leaveTangent=0.5, time=0.2, value=0.95}
          <at 0.3 1.05 RCIM_Cubic>
          {arriveTangent=-0.5, interpMode=RCIM_Cubic, leaveTangent=-0.5, time=0.4, value=0.95}
          {arriveTangent=0.25, interpMode=RCIM_Cubic, leaveTangent=0.25, time=0.5, value=0.95}
          {arriveTangent=0.5, interpMode=RCIM_Cubic, leaveTangent=0.5, time=0.6, value=1}
          {arriveTangent=-0.1667, interpMode=RCIM_Cubic, leaveTangent=-0.1667, time=0.7, value=1.05}
          {arriveTangent=-0.1667, interpMode=RCIM_Cubic, leaveTangent=-0.1667, time=0.9, value=0.95}
          {arriveTangent=0.5, interpMode=RCIM_Cubic, leaveTangent=0.5, time=1, value=1}
         ]}}}
      Scale Curve=50
     ))
    Write to Module Namespace=true
   )
   &SpriteRotationRate(Rotation Rate=30)
   &FN_Light_Attributes(
    Color=$Particles.Color
    Create Color Attribute=true
    Create Exponent Attribute=true
    Create Position Attribute=true
    Create Radius Attribute=true
    Exponent=6
    Position=<0 -25 0>
    Radius=50
    Scale Color=&FloatFromCurve(
     FloatCurve=DataInterface() {propertyValues {Curve {keys [
         {arriveTangent=-0.5, interpMode=RCIM_Cubic, leaveTangent=-0.5, time=0, value=1}
         {arriveTangent=-0.25, interpMode=RCIM_Cubic, leaveTangent=-0.25, time=0.1, value=0.95}
         {arriveTangent=0.5, interpMode=RCIM_Cubic, leaveTangent=0.5, time=0.2, value=0.95}
         <at 0.3 1.05 RCIM_Cubic>
         {arriveTangent=-0.5, interpMode=RCIM_Cubic, leaveTangent=-0.5, time=0.4, value=0.95}
         {arriveTangent=0.25, interpMode=RCIM_Cubic, leaveTangent=0.25, time=0.5, value=0.95}
         {arriveTangent=0.5, interpMode=RCIM_Cubic, leaveTangent=0.5, time=0.6, value=1}
         {arriveTangent=-0.1667, interpMode=RCIM_Cubic, leaveTangent=-0.1667, time=0.7, value=1.05}
         {arriveTangent=-0.1667, interpMode=RCIM_Cubic, leaveTangent=-0.1667, time=0.9, value=0.95}
         {arriveTangent=0.5, interpMode=RCIM_Cubic, leaveTangent=0.5, time=1, value=1}
        ]}}}
     Scale Curve=2
    )
   )
  ]
  rendererClasses [
   &NiagaraSpriteRendererProperties
   &NiagaraLightRendererProperties
   &NiagaraMeshRendererProperties
  ]
  renderers [
   {rendererClass=&NiagaraSpriteRendererProperties}
   {rendererClass=&NiagaraLightRendererProperties, rendererIndex=1}
   {rendererClass=&NiagaraMeshRendererProperties, rendererIndex=2}
  ]
 }
 Sparks_World {
  emitterUpdateScript [
   &EmitterState()
   &SpawnRate(SpawnRate=25, Use Spawn Probability=false)
  ]
  particleSpawnScript [
   &InitializeParticle(
    Lifetime Max=1
    Lifetime Min=0.5
    Lifetime Mode=::NewEnumerator1
    Sprite Rotation Mode=::NewEnumerator2
    Sprite Size=&Multiply_Vector2D_ByFloat(Float=&RandomRangeFloat(Minimum=1.5, Maximum=0.2), Vector2D=<2 2>)
    Sprite Size Mode=::NewEnumerator4
    UsePositionOffset=false
    Write Material Random=true
   )
   &SkeletalMeshLocation(
    MaskSpawnGroup=false
    Mesh Sampling Type=::NewEnumerator5
    Position Sampling=::NewEnumerator1
    Sampled Position Offset=&RandomRangeVector(Minimum=<-5 1 -4>, Maximum=<5 1 -10>)
    Skeletal Mesh=$User.New Skeletal Mesh
    Socket Sampling Mode=::NewEnumerator5
   )
   &InheritVelocity(Inherited Velocity Amount Scale=<0.5 0.5 0.5>, Speed Limit=true, UseSpeedThreshold=false)
   &AddVelocity(Velocity=&RandomRangeVector(Minimum=<0 0 0>, Maximum=<0 -15 0>))
   &AddVelocityFromPoint(
    Module.Use Falloff Distance=false
    Velocity Origin:NiagaraPosition=$Output.SkeletalMeshLocation.SampledPositionWithoutOffset
    Velocity Strength=13
   )
  ]
  particleUpdateScript [
   &ParticleState()
   &Drag(Use Linear Drag=true, Use Rotational Drag=false)
   &GravityForce(Gravity=<0 0 -35>)
   &InheritSourceMovement(Applied Position Scale=<0.5 0.5 0.5>, ApplyForce=false, ApplyPosition=true, ApplyPositionFalloff=false)
   &SolveForcesAndVelocity(Clamp Velocity=false, Limit Acceleration=false)
   &Color(
    Color=&ColorFromCurve(ColorCurve=DataInterface() {propertyValues {
       AlphaCurve {keys [<at 0 1 RCIM_Cubic>,<at 0.9977 1 RCIM_Cubic>]}
       BlueCurve {keys [
         {arriveTangent=-4.3342, interpMode=RCIM_Cubic, leaveTangent=-4.3342, time=0, value=0.6667}
         <at 0.1515 0.01>
         <at 0.306 1>
         <at 0.5971 1>
         <at 0.7436 1.5>
         {arriveTangent=0.3488, interpMode=RCIM_Cubic, leaveTangent=0.3488, time=1, value=1.5894}
        ]}
       GreenCurve {keys [
         {arriveTangent=-5.4563, interpMode=RCIM_Cubic, leaveTangent=-5.4563, time=0, value=1}
         <at 0.1515 0.1733>
         <at 0.306 0.6155>
         <at 0.5971 0.2083>
         <at 0.7436 0.075>
         {arriveTangent=0.0975, interpMode=RCIM_Cubic, leaveTangent=0.0975, time=1, value=0.1}
        ]}
       RedCurve {keys [
         {arriveTangent=3.3002, interpMode=RCIM_Cubic, leaveTangent=3.3002, time=0, value=0}
         <at 0.1515 0.5>
         <at 0.306 0.05>
         <at 0.5971 0.05>
         <at 0.7436 0.3125>
         {arriveTangent=6.5811, interpMode=RCIM_Cubic, leaveTangent=6.5811, time=1, value=2}
        ]}
      }})
    Scale Alpha=&FloatFromCurve(
     FloatCurve=DataInterface() {propertyValues {Curve {keys [
         {interpMode=RCIM_Cubic, tangentMode=RCTM_User, time=-0.0063, value=0.0184}
         {arriveTangent=-0.8483, interpMode=RCIM_Cubic, leaveTangent=-0.8483, tangentMode=RCTM_User, time=0.1718, value=1.7147}
         {arriveTangent=-5.0978, interpMode=RCIM_Cubic, leaveTangent=-5.0978, tangentMode=RCTM_User, time=0.2818, value=1.2143}
         {arriveTangent=-0.2783, interpMode=RCIM_Cubic, leaveTangent=-0.2783, tangentMode=RCTM_User, time=0.4, value=0.7}
         {arriveTangent=0.3112, interpMode=RCIM_Cubic, leaveTangent=0.3112, tangentMode=RCTM_User, time=0.5407, value=1.3419}
         {arriveTangent=-3.7119, interpMode=RCIM_Cubic, leaveTangent=-3.7119, tangentMode=RCTM_User, time=0.7242, value=0.3749}
         {interpMode=RCIM_Cubic, tangentMode=RCTM_User, time=1.0011, value=0.002}
        ]}}}
     Scale Curve=0.8
    )
    Scale Color=&VectorFromFloat(Value=50)
   )
   /SetVariables(Particles.Color2=&MultiplyLinearColor(Linear Color=$Particles.Color, Scale Factor=&MakeVector4_FromFloat(Float=0.1))) {isSetParameters=true}
   &ScaleSpriteSize(Scale Factor=&Vector2DFromFloat(Value=&FloatFromCurve(FloatCurve=DataInterface() {propertyValues {Curve {keys [
          {arriveTangent=7.1136, interpMode=RCIM_Cubic, leaveTangent=7.1136, tangentMode=RCTM_User, time=0, value=0}
          {arriveTangent=1.125, interpMode=RCIM_Cubic, leaveTangent=1.125, time=0.2, value=0.85}
          <at 0.4 0.45>
          {arriveTangent=-0.0952, interpMode=RCIM_Cubic, leaveTangent=-0.0952, tangentMode=RCTM_User, time=0.7849, value=0.7663}
          {arriveTangent=-1.238, interpMode=RCIM_Cubic, leaveTangent=-1.238, time=1, value=0.5}
         ]}}})))
   /SetVariables(Particles.SpriteSize2=&Multiply_Vector2D_ByFloat(Float=5, Vector2D=$Particles.SpriteSize)) {isSetParameters=true}
   &CameraOffset(Camera Offset Amount=3) {enabled=false}
  ]
  rendererClasses [&NiagaraSpriteRendererProperties]
  renderers [
   {rendererClass=&NiagaraSpriteRendererProperties}
   {rendererClass=&NiagaraSpriteRendererProperties, rendererIndex=1}
  ]
 }
 Sparks_Yellow {
  emitterUpdateScript [
   &EmitterState()
   &SpawnRate(SpawnRate=10, Use Spawn Probability=false)
  ]
  particleSpawnScript [
   &InitializeParticle(
    Lifetime Max=1.25
    Lifetime Min=0.8
    Lifetime Mode=::NewEnumerator1
    Sprite Rotation Mode=::NewEnumerator2
    Sprite Size=&Multiply_Vector2D_ByFloat(Float=&RandomRangeFloat(Minimum=1.5, Maximum=0.2), Vector2D=<2 2>)
    Sprite Size Mode=::NewEnumerator4
    UsePositionOffset=false
    Write Material Random=true
   )
   &SkeletalMeshLocation(
    MaskSpawnGroup=false
    Mesh Sampling Type=::NewEnumerator5
    Position Sampling=::NewEnumerator1
    Skeletal Mesh=$User.New Skeletal Mesh
    Socket Sampling Mode=::NewEnumerator5
   )
   &CylinderLocation(
    Cylinder Axis:Vector3f=<0 1 0>
    Cylinder Height=1
    Cylinder Radius=4
    MaskSpawnGroup=false
    Override Seed=false
    Rotation Mode:ENiagara_RotationMode=::NewEnumerator1
    Surface Only=true
    Use Endcaps In Surface Only Mode=false
    UseOwnerScale=true
   )
   &AddVelocityFromPoint(
    Module.Use Falloff Distance=false
    Velocity Origin:NiagaraPosition=$Output.SkeletalMeshLocation.SampledPositionWithoutOffset
    Velocity Strength=7
   )
   &AddVelocity(Velocity=&RandomRangeVector(Minimum=<0 0 0>, Maximum=<1 -10 1>))
  ]
  particleUpdateScript [
   &ParticleState()
   &Drag(Drag=0.5, Use Linear Drag=true, Use Rotational Drag=false)
   &SolveForcesAndVelocity(Clamp Velocity=false, Limit Acceleration=false)
   &Color(
    Color=&ColorFromCurve(ColorCurve=DataInterface() {propertyValues {
       AlphaCurve {keys [<at 0 1 RCIM_Cubic>,<at 0.9977 1 RCIM_Cubic>]}
       BlueCurve {keys [
         {arriveTangent=-6.5343, interpMode=RCIM_Cubic, leaveTangent=-6.5343, time=0, value=1}
         <at 0.1515 0.01>
         <at 0.306 1>
         <at 0.5971 0.111>
         <at 0.7436 0.8662>
         {arriveTangent=4.4218, interpMode=RCIM_Cubic, leaveTangent=4.4218, time=1, value=2}
        ]}
       GreenCurve {keys [
         {arriveTangent=-0.8377, interpMode=RCIM_Cubic, leaveTangent=-0.8377, time=0, value=0.3003}
         <at 0.1515 0.1733>
         <at 0.306 0.1556>
         <at 0.5971 0.7965>
         <at 0.7436 1.3204>
         {arriveTangent=-3.2457, interpMode=RCIM_Cubic, leaveTangent=-3.2457, time=1, value=0.4882}
        ]}
       RedCurve {keys [
         {arriveTangent=0.2967, interpMode=RCIM_Cubic, leaveTangent=0.2967, time=0, value=0.4551}
         <at 0.1515 0.5>
         <at 0.306 0.4277>
         <at 0.5971 1>
         <at 0.7436 1.5>
         {arriveTangent=-2.1703, interpMode=RCIM_Cubic, leaveTangent=-2.1703, time=1, value=0.9435}
        ]}
      }})
    Scale Alpha=&FloatFromCurve(
     FloatCurve=DataInterface() {propertyValues {Curve {keys [
         {interpMode=RCIM_Cubic, tangentMode=RCTM_User, time=-0.0063, value=0.0184}
         {arriveTangent=-0.8483, interpMode=RCIM_Cubic, leaveTangent=-0.8483, tangentMode=RCTM_User, time=0.1718, value=1.7147}
         {arriveTangent=-5.0978, interpMode=RCIM_Cubic, leaveTangent=-5.0978, tangentMode=RCTM_User, time=0.2818, value=1.2143}
         {arriveTangent=-0.2783, interpMode=RCIM_Cubic, leaveTangent=-0.2783, tangentMode=RCTM_User, time=0.4, value=0.7}
         {arriveTangent=0.3112, interpMode=RCIM_Cubic, leaveTangent=0.3112, tangentMode=RCTM_User, time=0.5407, value=1.3419}
         {arriveTangent=-3.7119, interpMode=RCIM_Cubic, leaveTangent=-3.7119, tangentMode=RCTM_User, time=0.7242, value=0.3749}
         {interpMode=RCIM_Cubic, tangentMode=RCTM_User, time=1.0011, value=0.002}
        ]}}}
     Scale Curve=0.8
    )
    Scale Color=&VectorFromFloat(Value=50)
   )
   &ScaleSpriteSize(Scale Factor=&Vector2DFromFloat(Value=&FloatFromCurve(FloatCurve=DataInterface() {propertyValues {Curve {keys [
          {arriveTangent=7.1136, interpMode=RCIM_Cubic, leaveTangent=7.1136, tangentMode=RCTM_User, time=0, value=0}
          {arriveTangent=1.125, interpMode=RCIM_Cubic, leaveTangent=1.125, time=0.2, value=0.85}
          <at 0.4 0.45>
          {arriveTangent=-0.0952, interpMode=RCIM_Cubic, leaveTangent=-0.0952, tangentMode=RCTM_User, time=0.7849, value=0.7663}
          {arriveTangent=-1.238, interpMode=RCIM_Cubic, leaveTangent=-1.238, time=1, value=0.5}
         ]}}})))
   &CameraOffset(Camera Offset Amount=3) {enabled=false}
  ]
  rendererClasses [&NiagaraSpriteRendererProperties]
  renderers [{rendererClass=&NiagaraSpriteRendererProperties}]
 }
}
summary {
 systemName=NS_Backpack_Sunshine
 emitters [
  {
   emitterName=Light
   rendererClasses [
    &NiagaraSpriteRendererProperties
    &NiagaraLightRendererProperties
    &NiagaraMeshRendererProperties
   ]
  }
  {
   emitterName=FN_RadialSprite
   rendererClasses [&NiagaraSpriteRendererProperties, &NiagaraMeshRendererProperties]
  }
  {emitterName=Sparks_World, rendererClasses [&NiagaraSpriteRendererProperties]}
  {emitterName=Sparks_Yellow, rendererClasses [&NiagaraSpriteRendererProperties]}
 ]
 userVariables [
  User.bIsFrontEndPreview {type=NiagaraBool}
  User.New Skeletal Mesh {
   type=NiagaraDataInterfaceSkeletalMesh
   defaultValue {
    struct=NiagaraExt_VariableValue_DataInterface
    value {dataInterface=&self:NiagaraDataInterfaceSkeletalMesh_1}
   }
  }
 ]
}
