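# Physics asset for the F_MED_Sunshine_Pack backpack cosmetic, defining collision and joint limits for its dangling handle and zipper-strap parts.
# All bodies use capsules, with two forming the pack and a much wider one at the strap tip, Dyn_Strap_03.
# The handle joins the pack directly, while a chain runs from pack through zipper and three strap bodies.
# Every joint sets swing limits with tighter twist limits, restricting how far parts bend and spin.
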PhysicsAsset /BRCosmetics/Accessories/FORT_Backpacks/F_MED_Sunshine_Pack/Meshes/F_MED_Sunshine_Pack_PhysicsAsset
bodies {
 Dyn_Handle [{shapeType=Capsule, length=3.299, radius=1.7022, shapeName=Dyn_Handle_capsule, center=<-1.1334 0 0>}]
 Dyn_Strap [{shapeType=Capsule, length=1.5, radius=0.3, shapeName=Dyn_Strap_capsule, center=<1 0 0>, rotation=<rot 90 0 0>}]
 Dyn_Strap_02 [{shapeType=Capsule, length=1, radius=0.3, shapeName=Dyn_Strap_02_capsule, center=<1 0 0>, rotation=<rot 90 0 0>}]
 Dyn_Strap_03 [{shapeType=Capsule, length=1, radius=2.75, shapeName=Dyn_Strap_03_capsule, center=<0.3 0 0>, rotation=<rot 90 0 90>}]
 Dyn_Zipper [{shapeType=Capsule, length=2, radius=0.3, shapeName=Dyn_Zipper_capsule, center=<1.5 0 0>, rotation=<rot 90 0 0>}]
 pack [
  {shapeType=Capsule, length=7.8249, radius=8.2619, center=<0 -3.0673 -3.2645>}
  {shapeType=Capsule, length=15, radius=5.25, shapeName=pack_capsule, center=<-11.1177 7.4184 -8.6037>, rotation=<rot 0 0 10.0023>}
 ]
}
constraints [
 {bone1Name=Dyn_Handle, bone2Name=pack, swing1LimitDegrees=35, swing2LimitDegrees=40, twistLimitDegrees=10}
 {bone1Name=Dyn_Strap_03, bone2Name=Dyn_Strap_02, swing1LimitDegrees=15, swing2LimitDegrees=20, twistLimitDegrees=3}
 {bone1Name=Dyn_Strap_02, bone2Name=Dyn_Strap, swing1LimitDegrees=25, swing2LimitDegrees=25, twistLimitDegrees=3}
 {bone1Name=Dyn_Strap, bone2Name=Dyn_Zipper, swing1LimitDegrees=25, swing2LimitDegrees=30, twistLimitDegrees=3}
 {bone1Name=Dyn_Zipper, bone2Name=pack, swing1LimitDegrees=20, swing2LimitDegrees=65, twistLimitDegrees=3}
]
