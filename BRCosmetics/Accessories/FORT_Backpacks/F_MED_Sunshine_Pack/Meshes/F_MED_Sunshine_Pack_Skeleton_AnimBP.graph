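# Animation Blueprint for the Sunshine backpack cosmetic that adds secondary physics motion to its dangling bones.
# The live chain starts from the local ref pose and runs a RigidBody node on the pack's physics asset, simulated relative to the pack bone with overridden gravity.
# A small additive offset on Dyn_Strap follows before output.
# Anim Dynamics nodes for zipper, strap and handle, plus a Dyn_Strap_03 transform, are disconnected from output; the event graph is empty.
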AnimBlueprint /BRCosmetics/Accessories/FORT_Backpacks/F_MED_Sunshine_Pack/Meshes/F_MED_Sunshine_Pack_Skeleton_AnimBP
target_skeleton=&/F_MED_Sunshine_Pack_Skeleton
graphs [
 AnimGraph {nodes [
   n1 = Anim Dynamics - Bone: Dyn_Zipper(Alpha=1, ExternalForce="0,0,0") {
    bGravityOverrideInSimSpace=true
    bOverrideAngularDamping=true
    bUseGravityOverride=true
    simulationSpace=RootRelative
    alphaBoolBlend {blendOption=Linear}
    boundBone {boneName=Dyn_Zipper}
    componentAppliedLinearAccClamp=<3000 3000 3000>
    componentLinearAccScale=<0.3 0.3 0.4>
    componentLinearVelScale=<0.3 0.3 0.3>
    gravityOverride {z=-800}
    physicsBodyDefinitions [{
      sphereCollisionRadius=0
      boundBone {boneName=Dyn_Zipper}
      boxExtents=<3 3 7>
      constraintSetup {angularLimitsMax=<3 10 5>, angularLimitsMin=<-3 0 -5>, angularTarget=<0 0 0>}
      localJointOffset=<3 0 0>
     }]
    relativeSpaceBone {boneName=pack}
   }
   Anim Dynamics - Bone: Dyn_Strap(ComponentPose=@n1, Alpha=1, ExternalForce="0,0,0") {
    angularDampingOverride=0.7
    bAngularSpring=true
    bGravityOverrideInSimSpace=true
    bOverrideAngularDamping=true
    bUseGravityOverride=true
    simulationSpace=RootRelative
    alphaBoolBlend {blendOption=Linear}
    boundBone {boneName=Dyn_Strap}
    componentAppliedLinearAccClamp=<3000 3000 3000>
    componentLinearAccScale=<0.3 0.3 0.4>
    componentLinearVelScale=<0.3 0.3 0.3>
    gravityOverride {z=-500}
    physicsBodyDefinitions [{
      sphereCollisionRadius=0
      boundBone {boneName=Dyn_Strap}
      boxExtents=<3 3 5>
      constraintSetup {angularLimitsMax=<30 20 60>, angularLimitsMin=<-30 -20 -30>, angularTarget=<0 0 0>}
      localJointOffset=<4 0 0>
     }]
    relativeSpaceBone {boneName=pack}
   }
   Anim Dynamics - Bone: Dyn_Handle(Alpha=1, ExternalForce="0,0,0") {
    bAngularSpring=true
    bGravityOverrideInSimSpace=true
    bOverrideAngularDamping=true
    bUseGravityOverride=true
    linearSpringConstant=1
    simulationSpace=RootRelative
    alphaBoolBlend {blendOption=Linear}
    boundBone {boneName=Dyn_Handle}
    componentAppliedLinearAccClamp=<3000 3000 3000>
    componentLinearAccScale=<0.3 0.3 0.4>
    componentLinearVelScale=<0.3 0.3 0.3>
    gravityOverride {z=-900}
    physicsBodyDefinitions [{
      sphereCollisionRadius=0
      boundBone {boneName=Dyn_Handle}
      boxExtents=<3 3 5>
      constraintSetup {angularLimitsMax=<7 10 15>, angularLimitsMin=<-7 -10 -15>, angularTarget=<0 0 0>}
      localJointOffset=<4 0 0>
     }]
    relativeSpaceBone {boneName=pack}
   }
   "Transform (Modify) Bone - Bone: Dyn_Strap_03"(Alpha=1, Translation="0,40,0", Rotation="90,0,0", Scale="1,1,1") {
    translationMode=BMM_Additive
    alphaBoolBlend {blendOption=Linear}
    boneToModify {boneName=Dyn_Strap_03}
    rotation {pitch=90}
    translation {y=40}
   }
   n2 = RigidBody(ComponentPose=Local To Component(LocalPose=Local Space Ref Pose())) {
    bClampLinearTranslationLimitToRefPose=true
    bOverrideWorldGravity=true
    overridePhysicsAsset=&/F_MED_Sunshine_Pack_PhysicsAsset
    simulationSpace=BaseBoneSpace
    alphaBoolBlend {blendOption=Linear}
    baseBoneRef {boneName=pack}
    componentAppliedLinearAccClamp=<5000 5000 5000>
    componentLinearAccScale=<0.3 0.3 0.35>
    componentLinearVelScale=<0.15 0.15 0.3>
    overrideWorldGravity {z=-980}
   }
   n3 = "Transform (Modify) Bone - Bone: Dyn_Strap"(ComponentPose=@n2, Alpha=1, Translation="0,0.1,0.15", Rotation="5,0,0", Scale="1,1,1") {
    translationMode=BMM_Additive
    alphaBoolBlend {blendOption=Linear}
    boneToModify {boneName=Dyn_Strap}
    rotation {pitch=5}
    translation {y=0.1, z=0.15}
   }
   Output Pose(Result=Component To Local(ComponentPose=@n3)) {layerGroup=None}
  ]}
 EventGraph {nodes [
   TryGetPawnOwner()
   Event BlueprintUpdateAnimation()
  ]}
]
variables []
